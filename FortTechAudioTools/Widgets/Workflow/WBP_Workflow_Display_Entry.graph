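# List entry widget for workflow displays in the Fort audio tech tools, showing a collapsible, titled section that hosts another widget.
# When a list item is set, it casts the item to its own class, so another instance of this widget serves as the data.
# It clears the body, then copies that item's SectionName to the title and adds its WidgetToShow to the body.
# An EditorUtilityExpandableArea, expanded by default, holds both title and body.
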WidgetBlueprint /FortTechAudioTools/Widgets/Workflow/WBP_Workflow_Display_Entry
graphs [
 EventGraph {nodes [
   n1 = Event OnListItemObjectSet()
   > n2 = Cast To WBP_Workflow_Display_Entry!(execute, Object=@n1.ListItemObject) -> (then, CastFailed, AsWBP Workflow Display Entry:WBP Workflow Display Entry Object Reference) {is=title K2Node_DynamicCast}
   > ClearChildren(self=Get VerticalBox_Body() -> (VerticalBox_Body:Vertical Box Object Reference))
   > SetSectionName!(execute, self, NewName:Text=Get SectionName!(self:WBP Workflow Display Entry Object Reference=@n2) -> (SectionName:Text)) -> (then)
   > SetWidget!(
    execute
    self
    Widget:User Widget Object Reference=Get WidgetToShow!(self:WBP Workflow Display Entry Object Reference=@n2) -> (WidgetToShow:User Widget Object Reference)
   ) -> (then)
   Event PreConstruct()
   Event Construct()
   Event Tick()
  ]}
 SetSectionName {nodes [
   n1 = SetSectionName!() -> (then, NewName:Text) {is=title K2Node_FunctionEntry}
   > Set Text(Text=@n1, self=Get CommonTextBlock_36() -> (CommonTextBlock_36:Common Text Object Reference))
  ]}
 SetWidget {nodes [
   n1 = SetWidget!() -> (then, Widget:User Widget Object Reference) {is=title K2Node_FunctionEntry}
   > AddChildToVerticalBox(self=Get VerticalBox_Body(), Content=@n1)
  ]}
]
variables [
 SectionName {type=string, default=""}
 WidgetToShow {type=UserWidget, default=null}
]
widgets {
 info {inheritedWidgetCount=0, namedSlotCount=2, parentClass=&UserWidget, rootWidgetClass=&CanvasPanel, widgetCount=5}
 widgets [
  CanvasPanel_24 {widgetClassPath=&CanvasPanel, widget=CanvasPanel_24}
  Border_0 {
   widgetClassPath=&Border
   parent=CanvasPanel_24
   widget=Border_0
   properties {
    background {
     bIsDynamicallyLoaded=false
     drawAs=Image
     imageType=NoImage
     mirroring=NoMirror
     resourceName=None
     resourceObject=null
     tiling=NoTile
     imageSize=<32 32>
     margin=<ltrb 0 0 0 0>
     outlineSettings {
      bUseBrushTransparency=false
      roundingType=HalfHeightRadius
      width=0
      color {colorUseRule=UseColor_Specified, specifiedColor=<rgba 0 0 0 0>}
      cornerRadii=<0 0 0 0>
     }
     tintColor {colorUseRule=UseColor_Specified, specifiedColor=<rgba 1 1 1 0>}
     uVRegion {bIsValid=false, max=<0 0>, min=<0 0>}
    }
    padding=<ltrb 0 0 0 0>
   }
   slot {
    class=CanvasPanelSlot
    properties {bAutoSize=true, layoutData {alignment=<0 0>, anchors {maximum=<1 0>, minimum=<0 0>}, offsets=<ltrb 0 0 0 0>}}
   }
  }
  EditorUtilityExpandableArea_64 {
   widgetClassPath=/Script/Blutility.EditorUtilityExpandableArea
   bIsVariable=true
   parent=Border_0
   widget=EditorUtilityExpandableArea_64
   properties {
    bIsExpanded=true
    areaPadding=<ltrb 25 0 0 0>
    borderBrush {
     bIsDynamicallyLoaded=false
     drawAs=Box
     imageType=NoImage
     mirroring=NoMirror
     resourceName=None
     resourceObject=null
     tiling=NoTile
     imageSize=<0 0>
     margin=<ltrb 0 0 0 0>
     outlineSettings {
      bUseBrushTransparency=false
      roundingType=FixedRadius
      width=0
      color {colorUseRule=UseColor_Specified, specifiedColor=<rgba 0 0 0 0>}
      cornerRadii=<4 4 4 4>
     }
     tintColor {colorUseRule=UseColor_Specified, specifiedColor=<rgba 0.0212 0.0212 0.0212 1>}
     uVRegion {bIsValid=false, max=<0 0>, min=<0 0>}
    }
   }
   slot {class=BorderSlot, properties {padding=<ltrb 0 0 0 0>}}
  }
  CommonTextBlock_36 {
   widgetClassPath=&CommonTextBlock
   bIsVariable=true
   namedSlotHost=EditorUtilityExpandableArea_64
   widget=CommonTextBlock_36
   properties {
    style=/FortUILibrary/Styles/Typography/Foundation/TextStyle-Base-XS.TextStyle-Base-XS_C
    text=Section
    font {
     bForceMonospaced=false
     bMaterialIsStencil=false
     fontMaterial=null
     fontObject=&NotoSans
     letterSpacing=0
     monospacedWidth=1
     size=16
     skewAmount=0
     typefaceFontName=NotoSans-Regular
     outlineSettings {
      bApplyOutlineToDropShadows=false
      bMiteredCorners=false
      bSeparateFillAlpha=false
      outlineMaterial=null
      outlineSize=0
      outlineColor=<rgba 0 0 0 1>
     }
    }
    shadowOffset=<0 0>
   }
  }
  VerticalBox_Body {widgetClassPath=&VerticalBox, bIsVariable=true, namedSlotHost=EditorUtilityExpandableArea_64, widget=VerticalBox_Body}
 ]
}
