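# Editor utility widget in FortTechAudioTools that displays the workflow widgets defined by a picked BP_WorkflowHelperData asset.
# Changing the asset broadcasts OnDataAssetChanged, opens its on-load utility widgets as tabs keyed by class path, then rebuilds the list.
# The rebuild wraps each mapped widget in a named entry, skipping BPI_WorkflowItem widgets whose Initialize fails.
# SetPropertyView sets the asset externally and only rebuilds the list.
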EditorUtilityWidgetBlueprint /FortTechAudioTools/Widgets/Workflow/EUW_Workflow_Display
graphs [
 EventGraph {nodes [
   "On Property Changed (SinglePropertyView_38)"!() -> (OutputDelegate, then, PropertyName:Name=None) {is=title K2Node_ComponentBoundEvent}
   n1 = Get WorkFlowHelperData() -> (WorkFlowHelperData:BP Workflow Helper Data Object Reference)
   > Call On Data Asset Changed!(
    execute:Exec
    self:EUW Workflow Display Object Reference
    Data:Object Reference=@n1
    TabID:Guid Structure=Get TabID() -> (TabID:Guid Structure)
   ) -> (then:Exec) {is=title K2Node_CallDelegate}
   > n2 = For Each Loop(Array:Array of User Widget Soft Class References=Get EditorUtilityWidgetsToRunOnLoad!(self:BP Workflow Helper Data Object Reference=@n1) -> (EditorUtilityWidgetsToRunOnLoad:Array of User Widget Soft Class References)) -> (Array Element:User Widget Soft Class Reference)
   > n3 = Async Load Class Asset!(execute:Exec, AssetClass:Object Soft Class Reference=@n2) -> (then:Exec, Completed:Exec, Class:Object Class Reference) {is=title K2Node_LoadAssetClass}
   OpenEditorUtilityWidgetBlueprint!(execute=@n3.Completed, self, Input:Object Class Reference=@n3) -> (then)
   ClearListItems(execute=[@n16.then, @n2.Completed], self=Get CommonListView_145() -> (CommonListView_145:Common List View Object Reference))
   > n4 = Get!(execute:Exec) -> (then:Exec, else:Exec, WorkFlowHelperData:BP Workflow Helper Data Object Reference)
   > n5 = GetWidgetsToLoad!(execute, self:BP Workflow Helper Data Object Reference=@n4) -> (then, WidgetsToLoad:Map of Strings to User Widget Soft Class References)
   > n6 = "For Each Loop (Map)"!(execute:Exec, MapPin:Map of Strings to User Widget Soft Class References=@n5, BreakPin:Exec) -> (then:Exec, KeyPin:String, ValuePin:User Widget Soft Class Reference, CompletedPin:Exec)
   > n7 = LoadClassAsset_Blocking(AssetClass=@n6.ValuePin)
   > n8 = Cast To UserWidget Class!(execute:Exec, Class:Object Class Reference=@n7) -> (then:Exec, CastFailed:Exec, AsUser Widget:User Widget Class Reference) {is=title K2Node_ClassDynamicCast}
   > n9 = Create Widget!(execute:Exec, WorldContextObject:Object Reference, Class:User Widget Class Reference=@n8, OwningPlayer:Player Controller Object Reference) -> (then:Exec, ReturnValue:User Widget Object Reference)
   > n10 = Cast To BPI_WorkflowItem!(execute, Object=@n9) -> (then, CastFailed, AsBPI Workflow Item:BPI Workflow Item Interface) {is=title K2Node_DynamicCast}
   > n11 = Initialize!(execute, self:BPI Workflow Item Interface=@n10, WorkflowHelperData:BP Workflow Helper Data Object Reference=Get WorkFlowHelperData()) -> (then, bInitialized:bool)
   > n12 = Branch(Condition=@n11)
   n13 = Create Widget!(
    execute=[@n12.then, @n10.CastFailed]
    WorldContextObject
    Class=&/WBP_Workflow_Display_Entry.WBP_Workflow_Display_Entry_C
    OwningPlayer
    SectionName:Text="To Text (String)"(InString=@n6)
    WidgetToShow:User Widget Object Reference=@n9
   ) -> (then, ReturnValue:WBP Workflow Display Entry Object Reference)
   > AddItem!(execute, self:List View Object Reference=Get CommonListView_145(), Item:Object Reference=@n13) -> (then)
   n14 = SetPropertyView!() -> (OutputDelegate, then, NewAsset:Object Reference) {is=title K2Node_CustomEvent}
   > n15 = Cast To BP_WorkflowHelperData!(execute, Object=@n14.NewAsset) -> (then, CastFailed, AsBP Workflow Helper Data:BP Workflow Helper Data Object Reference) {is=title K2Node_DynamicCast}
   > Set WorkFlowHelperData(WorkFlowHelperData:BP Workflow Helper Data Object Reference=@n15) -> (Output_Get:BP Workflow Helper Data Object Reference)
   > ForcePropertyChange()
   n16 = ForcePropertyChange() {is=title K2Node_CustomEvent}
   Event PreConstruct()
   > SetObject(self=Get SinglePropertyView_38() -> (SinglePropertyView_38:Single Property View Object Reference), NewObject=Self-Reference())
  ]}
 OnDataAssetChanged {nodes [OnDataAssetChanged!() -> (then, Data:Object Reference, TabID:Guid Structure) {is=title K2Node_FunctionEntry}]}
 OpenEditorUtilityWidgetBlueprint {nodes [
   n1 = OpenEditorUtilityWidgetBlueprint!() -> (then, Input:Object Class Reference) {is=title K2Node_FunctionEntry}
   n2 = String To Name(InString=GetSoftClassPath!(Class:Object Class Reference=@n1) -> (ReturnValue_PathString:String))
   n3 = Get EditorUtilitySubsystem()
   > n4 = SpawnRegisteredTabByID!(execute, self:Editor Utility Subsystem Object Reference=@n3, NewTabID:Name=@n2) -> (then, ReturnValue:bool)
   > n5 = Branch(Condition=@n4)
   Branch(execute=@n5.else, Condition=ClassIsChildOf(TestClass=@n1, ParentClass=&EditorUtilityWidget))
   > SpawnAndRegisterTabWithIdGeneratedClass!(execute, self:Editor Utility Subsystem Object Reference=@n3, InGeneratedWidgetBlueprint:Object Class Reference=@n1, InTabID:Name=@n2) -> (then, ReturnValue:Editor Utility Widget Object Reference)
  ]}
]
variables [
 LastKnownPath {type=string, default=""}
 OnDataAssetChanged {default=()}
 TabID {type=Guid, default=#1}
 WorkFlowHelperData {type=BP_WorkflowHelperData_C, default=null}
]
widgets {
 info {rootWidgetClass=&CanvasPanel, widgetCount=9}
 widgets [
  CanvasPanel_0 {widgetClassPath=&CanvasPanel, bIsVariable=true, widget=CanvasPanel_0}
  Border_0 {
   widgetClassPath=&Border
   parent=CanvasPanel_0
   widget=Border_0
   properties {background {
     bIsDynamicallyLoaded=false
     drawAs=Box
     imageType=NoImage
     mirroring=NoMirror
     resourceName=None
     resourceObject=null
     tiling=NoTile
     imageSize=<32 32>
     margin=<ltrb 1 1 1 1>
     outlineSettings {
      bUseBrushTransparency=false
      roundingType=HalfHeightRadius
      width=0
      color {colorUseRule=UseColor_Specified, specifiedColor=<rgba 0 0 0 0>}
      cornerRadii=<0 0 0 0>
     }
     tintColor {colorUseRule=UseColor_Specified, specifiedColor=<rgba 0.0212 0.0212 0.0212 1>}
     uVRegion {bIsValid=false, max=<0 0>, min=<0 0>}
    }}
   slot {
    class=CanvasPanelSlot
    properties {bAutoSize=true, layoutData {alignment=<0 0>, anchors {maximum=<1 1>, minimum=<0 0>}, offsets=<ltrb 0 0 0 0>}}
   }
  }
  VerticalBox_0 {widgetClassPath=&VerticalBox, parent=Border_0, widget=VerticalBox_0, slot {class=BorderSlot}}
  SinglePropertyView_38 {
   widgetClassPath=&SinglePropertyView
   parent=VerticalBox_0
   widget=SinglePropertyView_38
   properties {propertyName=WorkFlowHelperData}
   slot {class=VerticalBoxSlot, properties {horizontalAlignment=HAlign_Left, verticalAlignment=VAlign_Top, padding=<ltrb 10 10 0 10>}}
  }
  Border_1 {
   widgetClassPath=&Border
   parent=VerticalBox_0
   widget=Border_1
   properties {
    background {
     bIsDynamicallyLoaded=false
     drawAs=Box
     imageType=NoImage
     mirroring=NoMirror
     resourceName=None
     resourceObject=null
     tiling=NoTile
     imageSize=<32 32>
     margin=<ltrb 0 0 0 0>
     outlineSettings {
      bUseBrushTransparency=false
      roundingType=HalfHeightRadius
      width=0
      color {colorUseRule=UseColor_Specified, specifiedColor=<rgba 0 0 0 0>}
      cornerRadii=<0 0 0 0>
     }
     tintColor {colorUseRule=UseColor_Specified, specifiedColor=<rgba 0.0103 0.0103 0.0103 1>}
     uVRegion {bIsValid=false, max=<0 0>, min=<0 0>}
    }
    padding=<ltrb 0 0 0 0>
   }
   slot {class=VerticalBoxSlot, properties {padding=<ltrb 5 5 5 5>, size {sizeRule=Fill, value=1}}}
  }
  EditorUtilityScrollBox_0 {
   widgetClassPath=&EditorUtilityScrollBox
   parent=Border_1
   widget=EditorUtilityScrollBox_0
   slot {class=BorderSlot, properties {padding=<ltrb 0 0 0 0>}}
  }
  HorizontalBox_0 {
   widgetClassPath=&HorizontalBox
   parent=EditorUtilityScrollBox_0
   widget=HorizontalBox_0
   slot {class=ScrollBoxSlot, properties {size {sizeRule=Fill, value=1}}}
  }
  CommonListView_145 {
   widgetClassPath=&CommonListView
   bIsVariable=true
   parent=HorizontalBox_0
   widget=CommonListView_145
   properties {
    entryWidgetClass=&WBP_Workflow_Display_Entry_C
    scrollBarStyle {
     thickness=8
     draggedThumbImage {
      bIsDynamicallyLoaded=false
      drawAs=RoundedBox
      imageType=NoImage
      mirroring=NoMirror
      resourceName=None
      resourceObject=null
      tiling=NoTile
      imageSize=<0 0>
      margin=<ltrb 0 0 0 0>
      outlineSettings {
       bUseBrushTransparency=false
       roundingType=FixedRadius
       width=0
       color {colorUseRule=UseColor_Specified, specifiedColor=<rgba 0 0 0 0>}
       cornerRadii=<4 4 4 4>
      }
      tintColor {colorUseRule=UseColor_Specified, specifiedColor=<rgba 0.2159 0.2159 0.2159 1>}
      uVRegion {bIsValid=false, max=<0 0>, min=<0 0>}
     }
     horizontalBackgroundImage {
      bIsDynamicallyLoaded=false
      drawAs=NoDrawType
      imageType=NoImage
      mirroring=NoMirror
      resourceName=None
      resourceObject=null
      tiling=NoTile
      imageSize=<0 0>
      margin=<ltrb 0 0 0 0>
      outlineSettings {
       bUseBrushTransparency=false
       roundingType=HalfHeightRadius
       width=0
       color {colorUseRule=UseColor_Specified, specifiedColor=<rgba 0 0 0 0>}
       cornerRadii=<0 0 0 0>
      }
      tintColor {colorUseRule=UseColor_Specified, specifiedColor=<rgba 1 1 1 1>}
      uVRegion {bIsValid=false, max=<0 0>, min=<0 0>}
     }
     horizontalBottomSlotImage {
      bIsDynamicallyLoaded=false
      drawAs=NoDrawType
      imageType=NoImage
      mirroring=NoMirror
      resourceName=None
      resourceObject=null
      tiling=NoTile
      imageSize=<0 0>
      margin=<ltrb 0 0 0 0>
      outlineSettings {
       bUseBrushTransparency=false
       roundingType=HalfHeightRadius
       width=0
       color {colorUseRule=UseColor_Specified, specifiedColor=<rgba 0 0 0 0>}
       cornerRadii=<0 0 0 0>
      }
      tintColor {colorUseRule=UseColor_Specified, specifiedColor=<rgba 1 1 1 1>}
      uVRegion {bIsValid=false, max=<0 0>, min=<0 0>}
     }
     horizontalTopSlotImage {
      bIsDynamicallyLoaded=false
      drawAs=NoDrawType
      imageType=NoImage
      mirroring=NoMirror
      resourceName=None
      resourceObject=null
      tiling=NoTile
      imageSize=<0 0>
      margin=<ltrb 0 0 0 0>
      outlineSettings {
       bUseBrushTransparency=false
       roundingType=HalfHeightRadius
       width=0
       color {colorUseRule=UseColor_Specified, specifiedColor=<rgba 0 0 0 0>}
       cornerRadii=<0 0 0 0>
      }
      tintColor {colorUseRule=UseColor_Specified, specifiedColor=<rgba 1 1 1 1>}
      uVRegion {bIsValid=false, max=<0 0>, min=<0 0>}
     }
     hoveredThumbImage {
      bIsDynamicallyLoaded=false
      drawAs=RoundedBox
      imageType=NoImage
      mirroring=NoMirror
      resourceName=None
      resourceObject=null
      tiling=NoTile
      imageSize=<0 0>
      margin=<ltrb 0 0 0 0>
      outlineSettings {
       bUseBrushTransparency=false
       roundingType=FixedRadius
       width=0
       color {colorUseRule=UseColor_Specified, specifiedColor=<rgba 0 0 0 0>}
       cornerRadii=<4 4 4 4>
      }
      tintColor {colorUseRule=UseColor_Specified, specifiedColor=<rgba 0.2159 0.2159 0.2159 1>}
      uVRegion {bIsValid=false, max=<0 0>, min=<0 0>}
     }
     normalThumbImage {
      bIsDynamicallyLoaded=false
      drawAs=RoundedBox
      imageType=NoImage
      mirroring=NoMirror
      resourceName=None
      resourceObject=null
      tiling=NoTile
      imageSize=<0 0>
      margin=<ltrb 0 0 0 0>
      outlineSettings {
       bUseBrushTransparency=false
       roundingType=FixedRadius
       width=0
       color {colorUseRule=UseColor_Specified, specifiedColor=<rgba 0 0 0 0>}
       cornerRadii=<4 4 4 4>
      }
      tintColor {colorUseRule=UseColor_Specified, specifiedColor=<rgba 0.0953 0.0953 0.0953 1>}
      uVRegion {bIsValid=false, max=<0 0>, min=<0 0>}
     }
     verticalBackgroundImage {
      bIsDynamicallyLoaded=false
      drawAs=NoDrawType
      imageType=NoImage
      mirroring=NoMirror
      resourceName=None
      resourceObject=null
      tiling=NoTile
      imageSize=<0 0>
      margin=<ltrb 0 0 0 0>
      outlineSettings {
       bUseBrushTransparency=false
       roundingType=HalfHeightRadius
       width=0
       color {colorUseRule=UseColor_Specified, specifiedColor=<rgba 0 0 0 0>}
       cornerRadii=<0 0 0 0>
      }
      tintColor {colorUseRule=UseColor_Specified, specifiedColor=<rgba 1 1 1 1>}
      uVRegion {bIsValid=false, max=<0 0>, min=<0 0>}
     }
     verticalBottomSlotImage {
      bIsDynamicallyLoaded=false
      drawAs=NoDrawType
      imageType=NoImage
      mirroring=NoMirror
      resourceName=None
      resourceObject=null
      tiling=NoTile
      imageSize=<0 0>
      margin=<ltrb 0 0 0 0>
      outlineSettings {
       bUseBrushTransparency=false
       roundingType=HalfHeightRadius
       width=0
       color {colorUseRule=UseColor_Specified, specifiedColor=<rgba 0 0 0 0>}
       cornerRadii=<0 0 0 0>
      }
      tintColor {colorUseRule=UseColor_Specified, specifiedColor=<rgba 1 1 1 1>}
      uVRegion {bIsValid=false, max=<0 0>, min=<0 0>}
     }
     verticalTopSlotImage {
      bIsDynamicallyLoaded=false
      drawAs=NoDrawType
      imageType=NoImage
      mirroring=NoMirror
      resourceName=None
      resourceObject=null
      tiling=NoTile
      imageSize=<0 0>
      margin=<ltrb 0 0 0 0>
      outlineSettings {
       bUseBrushTransparency=false
       roundingType=HalfHeightRadius
       width=0
       color {colorUseRule=UseColor_Specified, specifiedColor=<rgba 0 0 0 0>}
       cornerRadii=<0 0 0 0>
      }
      tintColor {colorUseRule=UseColor_Specified, specifiedColor=<rgba 1 1 1 1>}
      uVRegion {bIsValid=false, max=<0 0>, min=<0 0>}
     }
    }
   }
   slot {class=HorizontalBoxSlot, properties {size {sizeRule=Fill, value=1}}}
  }
  Border_48 {
   widgetClassPath=&Border
   parent=HorizontalBox_0
   widget=Border_48
   properties {
    background {
     bIsDynamicallyLoaded=false
     drawAs=Box
     imageType=NoImage
     mirroring=NoMirror
     resourceName=None
     resourceObject=null
     tiling=NoTile
     imageSize=<32 32>
     margin=<ltrb 0 0 0 0>
     outlineSettings {
      bUseBrushTransparency=false
      roundingType=HalfHeightRadius
      width=0
      color {colorUseRule=UseColor_Specified, specifiedColor=<rgba 0 0 0 0>}
      cornerRadii=<0 0 0 0>
     }
     tintColor {colorUseRule=UseColor_Specified, specifiedColor=<rgba 0.0212 0.0212 0.0212 1>}
     uVRegion {bIsValid=false, max=<0 0>, min=<0 0>}
    }
    padding=<ltrb 35 0 0 0>
   }
   slot {class=HorizontalBoxSlot}
  }
 ]
}
