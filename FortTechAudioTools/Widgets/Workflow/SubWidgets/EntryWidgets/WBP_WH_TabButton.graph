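# Closable tab button widget for the FortTechAudioTools workflow UI, with a text label and an X close button.
# Presses on the tab and close button fire OnTabButtonPressed and OnClosed, each carrying the TabID.
# SelectionChanged stores Selected, swaps the button between Style and SelectedStyle, and shows the close button only while selected.
# Hover events also reveal the close button and brighten text from grey to white; SetTabName and ClearTabName set the label.
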WidgetBlueprint /FortTechAudioTools/Widgets/Workflow/SubWidgets/EntryWidgets/WBP_WH_TabButton
graphs [
 ClearTabName {nodes [
   ClearTabName() {is=title K2Node_FunctionEntry}
   > Set Text(Text="LOCTEXT(\"Empty\")", self=Get ButtonText() -> (ButtonText:Common Text Object Reference))
  ]}
 EventGraph {nodes [
   "On Pressed (Button)"() {is=title K2Node_ComponentBoundEvent}
   > Call On Tab Button Pressed!(execute:Exec, self:WBP WH Tab Button Object Reference, TabID:Guid Structure=Get TabID() -> (TabID:Guid Structure)) -> (then:Exec) {is=title K2Node_CallDelegate}
   "On Pressed (Button_Close)"() {is=title K2Node_ComponentBoundEvent}
   > Call On Closed!(execute:Exec, self:WBP WH Tab Button Object Reference, TabID:Guid Structure=Get TabID()) -> (then:Exec) {is=title K2Node_CallDelegate}
   n1 = "On Hovered (Button)"() {is=title K2Node_ComponentBoundEvent}
   SetColorAndOpacity(execute=[@n3.then, @n1.then], self=Get ButtonText(), InColorAndOpacity=Make Slate Color(SpecifiedColor=(R=1,G=1,B=1,A=1)))
   > SetVisibility(self=Get Button_Close() -> (Button_Close:Button Object Reference), InVisibility=Visible)
   "On Unhovered (Button)"() {is=title K2Node_ComponentBoundEvent}
   > n2 = SetVisibility(self=Get Button_Close(), InVisibility=Select(Option 0=Hidden, Option 1=Visible, Index=Get Selected() -> (Selected:bool)))
   SetColorAndOpacity(
    execute=[@n2.then, @n4.then]
    self=[Get CloseText() -> (CloseText:Common Text Object Reference), Get ButtonText()]
    InColorAndOpacity=Make Slate Color(SpecifiedColor=(R=0.3467,G=0.3467,B=0.3467,A=1))
   )
   "On Hovered (Button_Close)"() {is=title K2Node_ComponentBoundEvent}
   > n3 = SetColorAndOpacity(self=Get CloseText(), InColorAndOpacity=Make Slate Color(SpecifiedColor=(R=1,G=1,B=1,A=1)))
   n4 = "On Unhovered (Button_Close)"() {is=title K2Node_ComponentBoundEvent}
   n5 = SelectionChanged!() -> (OutputDelegate, then, Selected:bool=~) {is=title K2Node_CustomEvent}
   > n6 = Set Selected(Selected:bool=@n5.Selected) -> (Output_Get:bool)
   > Set WidgetStyle!(
    execute
    WidgetStyle:Button Style Structure=Select(
     Option 0:Button Style Structure=Get Style() -> (Style:Button Style Structure)
     Option 1:Button Style Structure=Get SelectedStyle() -> (SelectedStyle:Button Style Structure)
     Index=@n6
    ) -> (ReturnValue:Button Style Structure)
    self:Button Object Reference=Get Button() -> (Button:Button Object Reference)
   ) -> (then, Output_Get:Button Style Structure)
   > SetVisibility(
    self=Get Button_Close()
    InVisibility=Select(Option 0:ESlateVisibility Enum=Hidden, Option 1:ESlateVisibility Enum=Visible, Index=Get Selected()) -> (ReturnValue:ESlateVisibility Enum)
   )
   Event Construct()
   > SetColorAndOpacity(self=[Get CloseText(), Get ButtonText()], InColorAndOpacity=Make Slate Color(SpecifiedColor=(R=0.3467,G=0.3467,B=0.3467,A=1)))
  ]}
 OnClosed {nodes [OnClosed!() -> (then, TabID:Guid Structure) {is=title K2Node_FunctionEntry}]}
 OnTabButtonPressed {nodes [OnTabButtonPressed!() -> (then, TabID:Guid Structure) {is=title K2Node_FunctionEntry}]}
 SetTabName {nodes [
   n1 = SetTabName!() -> (then, NewName:Text) {is=title K2Node_FunctionEntry}
   > Set Text(Text=@n1, self=Get ButtonText())
  ]}
]
variables [
 OnClosed {}
 OnTabButtonPressed {}
 Selected {type=boolean}
 SelectedStyle {
  type=ButtonStyle
  default {
   clickedSlateSound {resourceObject=null}
   disabled {
    bIsDynamicallyLoaded=false
    drawAs=NoDrawType
    imageType=NoImage
    mirroring=NoMirror
    resourceName=None
    resourceObject=null
    tiling=NoTile
    imageSize=<0 0>
    margin=<ltrb 0 0 0 0>
    outlineSettings {
     bUseBrushTransparency=false
     roundingType=HalfHeightRadius
     width=0
     color {colorUseRule=UseColor_Specified, specifiedColor=<rgba 0 0 0 0>}
     cornerRadii=<0 0 0 0>
    }
    tintColor {colorUseRule=UseColor_Specified, specifiedColor=<rgba 1 1 1 1>}
    uVRegion {bIsValid=false, max=<0 0>, min=<0 0>}
   }
   disabledForeground {colorUseRule=UseColor_Specified, specifiedColor=<rgba 0.5271 0.5271 0.5271 1>}
   hovered {
    bIsDynamicallyLoaded=false
    drawAs=Box
    imageType=NoImage
    mirroring=NoMirror
    resourceName=None
    resourceObject=null
    tiling=NoTile
    imageSize=<32 32>
    margin=<ltrb 0 0 0 0>
    outlineSettings {
     bUseBrushTransparency=true
     roundingType=FixedRadius
     width=1
     color {colorUseRule=UseColor_Specified, specifiedColor=<rgba 0.7243 0.7243 0.7243 1>}
     cornerRadii=<4 4 4 4>
    }
    tintColor {colorUseRule=UseColor_Specified, specifiedColor=<rgba 0.0194 0.0194 0.0194 1>}
    uVRegion {bIsValid=false, max=<0 0>, min=<0 0>}
   }
   hoveredForeground {colorUseRule=UseColor_Specified, specifiedColor=<rgba 0 1 1 1>}
   hoveredSlateSound {resourceObject=null}
   normal {
    bIsDynamicallyLoaded=false
    drawAs=Box
    imageType=NoImage
    mirroring=NoMirror
    resourceName=None
    resourceObject=null
    tiling=NoTile
    imageSize=<32 32>
    margin=<ltrb 0 0 0 0>
    outlineSettings {
     bUseBrushTransparency=true
     roundingType=FixedRadius
     width=1
     color {colorUseRule=UseColor_Specified, specifiedColor=<rgba 0.6951 0.6951 0.6951 1>}
     cornerRadii=<4 4 4 4>
    }
    tintColor {colorUseRule=UseColor_Specified, specifiedColor=<rgba 0.0194 0.0194 0.0194 1>}
    uVRegion {bIsValid=false, max=<0 0>, min=<0 0>}
   }
   normalForeground {colorUseRule=UseColor_Specified, specifiedColor=<rgba 0 1 1 1>}
   normalPadding=<ltrb 12 1.5 12 1.5>
   pressed {
    bIsDynamicallyLoaded=false
    drawAs=Box
    imageType=NoImage
    mirroring=NoMirror
    resourceName=None
    resourceObject=null
    tiling=NoTile
    imageSize=<32 32>
    margin=<ltrb 0 0 0 0>
    outlineSettings {
     bUseBrushTransparency=true
     roundingType=FixedRadius
     width=1
     color {colorUseRule=UseColor_Specified, specifiedColor=<rgba 0.7243 0.7243 0.7243 0>}
     cornerRadii=<0 0 0 0>
    }
    tintColor {colorUseRule=UseColor_Specified, specifiedColor=<rgba 0.0194 0.0194 0.0194 1>}
    uVRegion {bIsValid=false, max=<0 0>, min=<0 0>}
   }
   pressedForeground {colorUseRule=UseColor_Specified, specifiedColor=<rgba 0 1 1 1>}
   pressedPadding=<ltrb 12 1.5 12 1.5>
   pressedSlateSound {resourceObject=null}
  }
 }
 Style {
  type=ButtonStyle
  default {
   clickedSlateSound {resourceObject=null}
   disabled {
    bIsDynamicallyLoaded=false
    drawAs=NoDrawType
    imageType=NoImage
    mirroring=NoMirror
    resourceName=None
    resourceObject=null
    tiling=NoTile
    imageSize=<0 0>
    margin=<ltrb 0 0 0 0>
    outlineSettings {
     bUseBrushTransparency=false
     roundingType=HalfHeightRadius
     width=0
     color {colorUseRule=UseColor_Specified, specifiedColor=<rgba 0 0 0 0>}
     cornerRadii=<0 0 0 0>
    }
    tintColor {colorUseRule=UseColor_Specified, specifiedColor=<rgba 1 1 1 1>}
    uVRegion {bIsValid=false, max=<0 0>, min=<0 0>}
   }
   disabledForeground {colorUseRule=UseColor_Specified, specifiedColor=<rgba 0.5271 0.5271 0.5271 1>}
   hovered {
    bIsDynamicallyLoaded=false
    drawAs=Box
    imageType=NoImage
    mirroring=NoMirror
    resourceName=None
    resourceObject=null
    tiling=NoTile
    imageSize=<32 32>
    margin=<ltrb 0 0 0 0>
    outlineSettings {
     bUseBrushTransparency=true
     roundingType=FixedRadius
     width=1
     color {colorUseRule=UseColor_Specified, specifiedColor=<rgba 0.7243 0.7243 0.7243 1>}
     cornerRadii=<4 4 4 4>
    }
    tintColor {colorUseRule=UseColor_Specified, specifiedColor=<rgba 0.0194 0.0194 0.0194 1>}
    uVRegion {bIsValid=false, max=<0 0>, min=<0 0>}
   }
   hoveredForeground {colorUseRule=UseColor_Specified, specifiedColor=<rgba 0 1 1 1>}
   hoveredSlateSound {resourceObject=null}
   normal {
    bIsDynamicallyLoaded=false
    drawAs=Box
    imageType=NoImage
    mirroring=NoMirror
    resourceName=None
    resourceObject=null
    tiling=NoTile
    imageSize=<32 32>
    margin=<ltrb 0 0 0 0>
    outlineSettings {
     bUseBrushTransparency=true
     roundingType=FixedRadius
     width=1
     color {colorUseRule=UseColor_Specified, specifiedColor=<rgba 0.6951 0.6951 0.6951 1>}
     cornerRadii=<4 4 4 4>
    }
    tintColor {colorUseRule=UseColor_Specified, specifiedColor=<rgba 0.0048 0.0048 0.0048 1>}
    uVRegion {bIsValid=false, max=<0 0>, min=<0 0>}
   }
   normalForeground {colorUseRule=UseColor_Specified, specifiedColor=<rgba 0 1 1 1>}
   normalPadding=<ltrb 12 1.5 12 1.5>
   pressed {
    bIsDynamicallyLoaded=false
    drawAs=Box
    imageType=NoImage
    mirroring=NoMirror
    resourceName=None
    resourceObject=null
    tiling=NoTile
    imageSize=<32 32>
    margin=<ltrb 0 0 0 0>
    outlineSettings {
     bUseBrushTransparency=true
     roundingType=FixedRadius
     width=1
     color {colorUseRule=UseColor_Specified, specifiedColor=<rgba 0.7243 0.7243 0.7243 0>}
     cornerRadii=<0 0 0 0>
    }
    tintColor {colorUseRule=UseColor_Specified, specifiedColor=<rgba 0.0075 0.0075 0.0075 1>}
    uVRegion {bIsValid=false, max=<0 0>, min=<0 0>}
   }
   pressedForeground {colorUseRule=UseColor_Specified, specifiedColor=<rgba 0 1 1 1>}
   pressedPadding=<ltrb 12 1.5 12 1.5>
   pressedSlateSound {resourceObject=null}
  }
 }
 TabID {type=Guid, default=#1}
]
widgets {
 info {inheritedWidgetCount=0, parentClass=&UserWidget, rootWidgetClass=&CanvasPanel, widgetCount=9}
 widgets [
  CanvasPanel_0 {widgetClassPath=&CanvasPanel, widget=CanvasPanel_0}
  SizeBox_52 {
   widgetClassPath=&SizeBox
   parent=CanvasPanel_0
   widget=SizeBox_52
   properties {bOverride_HeightOverride=true, bOverride_WidthOverride=true, heightOverride=25, maxDesiredHeight=35, maxDesiredWidth=122, widthOverride=150}
   slot {
    class=CanvasPanelSlot
    properties {bAutoSize=true, layoutData {alignment=<0 0>, anchors {maximum=<0 0>, minimum=<0 0>}, offsets=<ltrb 0 0 0 0>}}
   }
  }
  Button {
   widgetClassPath=&Button
   parent=SizeBox_52
   widget=Button
   properties {widgetStyle {
     clickedSlateSound {resourceObject=null}
     disabled {
      bIsDynamicallyLoaded=false
      drawAs=NoDrawType
      imageType=NoImage
      mirroring=NoMirror
      resourceName=None
      resourceObject=null
      tiling=NoTile
      imageSize=<0 0>
      margin=<ltrb 0 0 0 0>
      outlineSettings {
       bUseBrushTransparency=false
       roundingType=HalfHeightRadius
       width=0
       color {colorUseRule=UseColor_Specified, specifiedColor=<rgba 0 0 0 0>}
       cornerRadii=<0 0 0 0>
      }
      tintColor {colorUseRule=UseColor_Specified, specifiedColor=<rgba 1 1 1 1>}
      uVRegion {bIsValid=false, max=<0 0>, min=<0 0>}
     }
     disabledForeground {colorUseRule=UseColor_Specified, specifiedColor=<rgba 0.5271 0.5271 0.5271 1>}
     hovered {
      bIsDynamicallyLoaded=false
      drawAs=Box
      imageType=NoImage
      mirroring=NoMirror
      resourceName=None
      resourceObject=null
      tiling=Horizontal
      imageSize=<32 32>
      margin=<ltrb 0 0 0 0>
      outlineSettings {
       bUseBrushTransparency=true
       roundingType=FixedRadius
       width=1
       color {colorUseRule=UseColor_Specified, specifiedColor=<rgba 0.7243 0.7243 0.7243 1>}
       cornerRadii=<4 4 4 4>
      }
      tintColor {colorUseRule=UseColor_Specified, specifiedColor=<rgba 0.0103 0.0103 0.0103 1>}
      uVRegion {bIsValid=false, max=<0 0>, min=<0 0>}
     }
     hoveredForeground {colorUseRule=UseColor_Specified, specifiedColor=<rgba 0 1 1 1>}
     hoveredSlateSound {resourceObject=null}
     normal {
      bIsDynamicallyLoaded=false
      drawAs=Box
      imageType=NoImage
      mirroring=NoMirror
      resourceName=None
      resourceObject=null
      tiling=NoTile
      imageSize=<32 32>
      margin=<ltrb 0 0 0 0>
      outlineSettings {
       bUseBrushTransparency=true
       roundingType=HalfHeightRadius
       width=-0.2416
       color {colorUseRule=UseColor_Specified, specifiedColor=<rgba 0.6951 0.6951 0.6951 1>}
       cornerRadii=<15.3983 4 4 4>
      }
      tintColor {colorUseRule=UseColor_Specified, specifiedColor=<rgba 0.0103 0.0103 0.0103 1>}
      uVRegion {bIsValid=false, max=<0 0>, min=<0 0>}
     }
     normalForeground {colorUseRule=UseColor_Specified, specifiedColor=<rgba 0 1 1 1>}
     normalPadding=<ltrb 12 1.5 12 1.5>
     pressed {
      bIsDynamicallyLoaded=false
      drawAs=RoundedBox
      imageType=NoImage
      mirroring=NoMirror
      resourceName=None
      resourceObject=null
      tiling=NoTile
      imageSize=<32 32>
      margin=<ltrb 0 0 0 0>
      outlineSettings {
       bUseBrushTransparency=true
       roundingType=FixedRadius
       width=1
       color {colorUseRule=UseColor_Specified, specifiedColor=<rgba 0.7243 0.7243 0.7243 1>}
       cornerRadii=<4 4 4 4>
      }
      tintColor {colorUseRule=UseColor_Specified, specifiedColor=<rgba 0.0116 0.0116 0.0116 1>}
      uVRegion {bIsValid=false, max=<0 0>, min=<0 0>}
     }
     pressedForeground {colorUseRule=UseColor_Specified, specifiedColor=<rgba 0 1 1 1>}
     pressedPadding=<ltrb 12 1.5 12 1.5>
     pressedSlateSound {resourceObject=null}
    }}
   slot {class=SizeBoxSlot}
  }
  SizeBox_0 {
   widgetClassPath=&SizeBox
   parent=Button
   widget=SizeBox_0
   properties {bOverride_HeightOverride=true, heightOverride=25}
   slot {class=ButtonSlot, properties {padding=<ltrb 0 0 10 0>}}
  }
  ScaleBox_0 {widgetClassPath=&ScaleBox, parent=SizeBox_0, widget=ScaleBox_0, properties {stretchDirection=DownOnly}, slot {class=SizeBoxSlot}}
  ButtonText {
   widgetClassPath=&CommonTextBlock
   bIsVariable=true
   parent=ScaleBox_0
   widget=ButtonText
   properties {
    bIsScrollingEnabled=false
    style=/FortUILibrary/Styles/Typography/Foundation/TextStyle-Base-S.TextStyle-Base-S_C
    text=Empty
    font {
     bForceMonospaced=false
     bMaterialIsStencil=false
     fontMaterial=null
     fontObject=&NotoSans
     letterSpacing=0
     monospacedWidth=1
     size=18
     skewAmount=0
     typefaceFontName=NotoSans-Regular
     outlineSettings {
      bApplyOutlineToDropShadows=false
      bMiteredCorners=false
      bSeparateFillAlpha=false
      outlineMaterial=null
      outlineSize=0
      outlineColor=<rgba 0 0 0 1>
     }
    }
    shadowOffset=<0 0>
   }
   slot {class=ScaleBoxSlot}
  }
  SizeBox_53 {
   widgetClassPath=&SizeBox
   parent=CanvasPanel_0
   widget=SizeBox_53
   properties {bOverride_HeightOverride=true, bOverride_WidthOverride=true, heightOverride=15, widthOverride=15}
   slot {
    class=CanvasPanelSlot
    properties {bAutoSize=true, zOrder=1, layoutData {alignment=<0 0>, anchors {maximum=<1 0>, minimum=<1 0>}, offsets=<ltrb -22 5.3718 0 0>}}
   }
  }
  Button_Close {
   widgetClassPath=&Button
   parent=SizeBox_53
   widget=Button_Close
   properties {widgetStyle {
     clickedSlateSound {resourceObject=null}
     disabled {
      bIsDynamicallyLoaded=false
      drawAs=NoDrawType
      imageType=NoImage
      mirroring=NoMirror
      resourceName=None
      resourceObject=null
      tiling=NoTile
      imageSize=<0 0>
      margin=<ltrb 0 0 0 0>
      outlineSettings {
       bUseBrushTransparency=false
       roundingType=HalfHeightRadius
       width=0
       color {colorUseRule=UseColor_Specified, specifiedColor=<rgba 0 0 0 0>}
       cornerRadii=<0 0 0 0>
      }
      tintColor {colorUseRule=UseColor_Specified, specifiedColor=<rgba 1 1 1 1>}
      uVRegion {bIsValid=false, max=<0 0>, min=<0 0>}
     }
     disabledForeground {colorUseRule=UseColor_Specified, specifiedColor=<rgba 0.5271 0.5271 0.5271 1>}
     hovered {
      bIsDynamicallyLoaded=false
      drawAs=Box
      imageType=NoImage
      mirroring=NoMirror
      resourceName=None
      resourceObject=null
      tiling=NoTile
      imageSize=<32 32>
      margin=<ltrb 0 0 0 0>
      outlineSettings {
       bUseBrushTransparency=true
       roundingType=FixedRadius
       width=1
       color {colorUseRule=UseColor_Specified, specifiedColor=<rgba 0.7243 0.7243 0.7243 1>}
       cornerRadii=<4 4 4 4>
      }
      tintColor {colorUseRule=UseColor_Specified, specifiedColor=<rgba 0.0176 0.0176 0.0176 1>}
      uVRegion {bIsValid=false, max=<0 0>, min=<0 0>}
     }
     hoveredForeground {colorUseRule=UseColor_Specified, specifiedColor=<rgba 0 1 1 1>}
     hoveredSlateSound {resourceObject=null}
     normal {
      bIsDynamicallyLoaded=false
      drawAs=Box
      imageType=NoImage
      mirroring=NoMirror
      resourceName=None
      resourceObject=null
      tiling=NoTile
      imageSize=<32 32>
      margin=<ltrb 0 0 0 0>
      outlineSettings {
       bUseBrushTransparency=true
       roundingType=FixedRadius
       width=1
       color {colorUseRule=UseColor_Specified, specifiedColor=<rgba 0.6951 0.6951 0.6951 1>}
       cornerRadii=<4 4 4 4>
      }
      tintColor {colorUseRule=UseColor_Specified, specifiedColor=<rgba 0.0176 0.0176 0.0176 1>}
      uVRegion {bIsValid=false, max=<0 0>, min=<0 0>}
     }
     normalForeground {colorUseRule=UseColor_Specified, specifiedColor=<rgba 0 1 1 1>}
     normalPadding=<ltrb 12 1.5 12 1.5>
     pressed {
      bIsDynamicallyLoaded=false
      drawAs=Box
      imageType=NoImage
      mirroring=NoMirror
      resourceName=None
      resourceObject=null
      tiling=NoTile
      imageSize=<32 32>
      margin=<ltrb 0 0 0 0>
      outlineSettings {
       bUseBrushTransparency=true
       roundingType=FixedRadius
       width=1
       color {colorUseRule=UseColor_Specified, specifiedColor=<rgba 0.7243 0.7243 0.7243 1>}
       cornerRadii=<4 4 4 4>
      }
      tintColor {colorUseRule=UseColor_Specified, specifiedColor=<rgba 0.0176 0.0176 0.0176 1>}
      uVRegion {bIsValid=false, max=<0 0>, min=<0 0>}
     }
     pressedForeground {colorUseRule=UseColor_Specified, specifiedColor=<rgba 0 1 1 1>}
     pressedPadding=<ltrb 12 1.5 12 1.5>
     pressedSlateSound {resourceObject=null}
    }}
   slot {class=SizeBoxSlot}
  }
  CloseText {
   widgetClassPath=&CommonTextBlock
   bIsVariable=true
   parent=Button_Close
   widget=CloseText
   properties {
    bIsScrollingEnabled=false
    justification=Center
    style=/FortUILibrary/Styles/Typography/Foundation/TextStyle-Base-L-B.TextStyle-Base-L-B_C
    text=X
    font {
     bForceMonospaced=false
     bMaterialIsStencil=false
     fontMaterial=null
     fontObject=&NotoSans
     letterSpacing=0
     monospacedWidth=1
     size=32
     skewAmount=0
     typefaceFontName=NotoSans-Bold
     outlineSettings {
      bApplyOutlineToDropShadows=false
      bMiteredCorners=false
      bSeparateFillAlpha=false
      outlineMaterial=null
      outlineSize=0
      outlineColor=<rgba 0 0 0 1>
     }
    }
    renderTransform {angle=0, scale=<0.25 0.25>, shear=<0 0>, translation=<0 -6>}
    shadowOffset=<0 0>
   }
   slot {class=ButtonSlot, properties {padding=<ltrb 0 0 0 0>}}
  }
 ]
}
