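# Editor utility widget for the Fortnite audio Workflow Helper: a tabbed panel of data-asset views plus an AudioBase plugin tracker.
# GUID-keyed tabs pair buttons with WidgetSwitcher displays, naming themselves from assigned BP_WorkflowHelperData assets.
# Tab layout and soft asset references persist through a WH_SaveGameConfig save slot, restored on Construct.
# A looping timer drives embedded Python that detects Content Browser path changes and lists the plugin's AudioBase dependencies from its .uplugin.
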EditorUtilityWidgetBlueprint /FortTechAudioTools/Widgets/Workflow/EUW_Workflow_Helper
graphs [
 ActivateTab {nodes [
   n1 = ActivateTab!() -> (then, TabID:Guid Structure) {is=title K2Node_FunctionEntry}
   n2 = Find(
    TargetMap:Map of Guid Structures to Struct WH Tab Info Structures=Get TabsByGuid() -> (TabsByGuid:Map of Guid Structures to Struct WH Tab Info Structures)
    Key:"Guid Structure (by ref)"=@n1
   ) -> (Value:Struct WH Tab Info Structure)
   > Branch(Condition=@n2.ReturnValue)
   > SetActiveWidget(
    self:Widget Switcher Object Reference=Get WidgetSwitcher_0() -> (WidgetSwitcher_0:Widget Switcher Object Reference)
    Widget=Break Struct WH Tab Info!(Struct_WH_TabInfo:"Struct WH Tab Info Structure (by ref)"=@n2) -> (
     TabID_16_#1:Guid Structure
     ContentWidget_12_#2:EUW Workflow Display Object Reference
     TabButtonWidget_8_#3:WBP WH Tab Button Object Reference
     AssignedDataAsset_15_#4:Object Reference
     SaveSlotName_22_#5:String
     SaveObj_25_#6:WH Save Game Config Object Reference
    ) {is=title K2Node_BreakStruct}.ContentWidget_12_#2
   )
   > n3 = "For Each Loop (Map)"!(execute:Exec, MapPin:Map of Guid Structures to Struct WH Tab Info Structures=Get TabsByGuid(), BreakPin:Exec) -> (then:Exec, KeyPin:Guid Structure, ValuePin:Struct WH Tab Info Structure, CompletedPin:Exec)
   n4 = Break Struct WH Tab Info(Struct_WH_TabInfo=@n3.ValuePin)
   > SelectionChanged!(
    execute
    self:WBP WH Tab Button Object Reference=@n4.TabButtonWidget_8_#3
    Selected:bool="Equal (Guid)"(A:"Guid Structure (by ref)"=@n4, B:"Guid Structure (by ref)"=Get TabID() -> (TabID:Guid Structure)) -> (ReturnValue:bool=~) {is=title K2Node_PromotableOperator}
   ) -> (then)
  ]}
 AddTab {nodes [
   n1 = AddTab!() -> (then, InGuid:Guid Structure) {is=title K2Node_FunctionEntry}
   > n2 = Branch(Condition=Is Valid Guid!(InGuid:"Guid Structure (by ref)"=@n1) -> (ReturnValue:bool))
   > n3 = Set NewGuid(NewGuid:Guid Structure=Get InGuid() -> (InGuid:Guid Structure)) -> (Output_Get:Guid Structure)
   n4 = Create Widget!(
    execute:Exec=[@n9.then, @n3.then]
    WorldContextObject:Object Reference=Self-Reference()
    Class:User Widget Class Reference=&/EUW_Workflow_Display.EUW_Workflow_Display_C
    OwningPlayer:Player Controller Object Reference
    TabID:Guid Structure=Get NewGuid() -> (NewGuid:Guid Structure)
   ) -> (then:Exec, ReturnValue:EUW Workflow Display Object Reference)
   > Bind Event to On Data Asset Changed(self:EUW Workflow Display Object Reference=@n4, Delegate=Create Event()) {is=title K2Node_AddDelegate}
   > AddChild(self=Get WidgetSwitcher_0(), Content=@n4)
   > n5 = Create Entry!(execute, self:Dynamic Entry Box Object Reference=Get Tabs() -> (Tabs:Dynamic Entry Box Object Reference)) -> (then, ReturnValue:User Widget Object Reference)
   > n6 = Cast To WBP_WH_TabButton!(execute, Object=@n5) -> (then, CastFailed, AsWBP WH Tab Button:WBP WH Tab Button Object Reference) {is=title K2Node_DynamicCast}
   > Set TabID!(execute, TabID:Guid Structure=Get NewGuid(), self:WBP WH Tab Button Object Reference=@n6) -> (then, Output_Get:Guid Structure)
   > ClearTabName(self:WBP WH Tab Button Object Reference=@n6)
   > Bind Event to On Tab Button Pressed(self:WBP WH Tab Button Object Reference=@n6, Delegate=Create Event()) {is=title K2Node_AddDelegate}
   > Bind Event to On Closed(self:WBP WH Tab Button Object Reference=@n6, Delegate=Create Event()) {is=title K2Node_AddDelegate}
   > Add!(
    execute
    TargetMap:Map of Guid Structures to Struct WH Tab Info Structures=Get TabsByGuid()
    Key:"Guid Structure (by ref)"=Get NewGuid()
    Value:"Struct WH Tab Info Structure (by ref)"=Make Struct WH Tab Info!(
     TabID_16_#1:Guid Structure=Get NewGuid()
     ContentWidget_12_#2:EUW Workflow Display Object Reference=@n4
     TabButtonWidget_8_#3:WBP WH Tab Button Object Reference=@n6
     AssignedDataAsset_15_#4:Object Reference
     SaveSlotName_22_#5:String
     SaveObj_25_#6:WH Save Game Config Object Reference
    ) -> (Struct_WH_TabInfo:Struct WH Tab Info Structure) {is=title K2Node_MakeStruct}
   ) -> (then) {is=title K2Node_CallFunction}
   > ActivateTab!(execute, self, TabID:Guid Structure=Get NewGuid()) -> (then) {is=title K2Node_CallFunction}
   > n7 = "For Each Loop (Map)"(MapPin=Get TabsByGuid())
   n8 = Break Struct WH Tab Info(Struct_WH_TabInfo=@n7.ValuePin)
   > Is Valid(InputObject=@n8.AssignedDataAsset_15_#4)
   > SetPropertyView!(
    execute
    self:EUW Workflow Display Object Reference=@n8.ContentWidget_12_#2
    NewAsset:Object Reference=@n8.AssignedDataAsset_15_#4
   ) -> (then)
   Return Node(execute=@n7.CompletedPin)
   n9 = Set NewGuid(execute=@n2.else, NewGuid=NewGuid!() -> (ReturnValue:Guid Structure))
  ]}
 EventGraph {nodes [
   "On Pressed (Button_AddTab)"() {is=title K2Node_ComponentBoundEvent}
   > AddTab!(execute, self, InGuid:Guid Structure) -> (then) {is=title K2Node_CallFunction}
   > SaveLayout()
   "On Pressed (EditorUtilityButton_86)"() {is=title K2Node_ComponentBoundEvent}
   > LaunchURL!(execute, URL:String=https://epicgames.atlassian.net/wiki/spaces/FORT/pages/326129197/Audio+Workflow+Helper) -> (then)
   ContentBrowserLocationCheck() {is=title K2Node_CustomEvent}
   > n1 = Execute Python Script!(
    execute
    PythonScript="import unreal\r\nimport json\r\n\r\n## Get current path\r\ncurrent_path = unreal.EditorUtilityLibrary.get_current_content_browser_path()\r\n\r\n## Initialize outputs\r\nhas_changed = False\r\naudio_plugins_string = []\r\nnew_last_path = last_known_path  ## Pass through the input\r\ncontent_browser_location = \"\"\r\n\r\n## Handle None case\r\nif current_path is None:\r\n    current_path = \"\"\r\n\r\ncontent_browser_location = current_path\r\n\r\n## Only check for changes if we have a valid path\r\nif current_path != \"\":\r\n    ## Check if changed compared to the Blueprint's stored path\r\n    if current_path != last_known_path:\r\n        has_changed = True\r\n        new_last_path = current_path  ## Update the path to store back in Blueprint\r\n\r\n        ## Get plugin name from the path\r\n        plugin_name = unreal.AnimBlueprintNativizationEditorLibrary.get_plugin_name_for_asset_path(current_path)\r\n\r\n        if plugin_name:\r\n            ## Get the plugin descriptor file path\r\n            descriptor_path = unreal.PluginBlueprintLibrary.get_plugin_descriptor_file_path(plugin_name)\r\n\r\n            if descriptor_path:\r\n                try:\r\n                    ## Read and parse the .uplugin JSON file\r\n                    with open(descriptor_path, 'r') as f:\r\n                        plugin_data = json.load(f)\r\n\r\n                    ## Check plugin dependencies\r\n                    if 'Plugins' in plugin_data:\r\n                        for dependency in plugin_data['Plugins']:\r\n                            dep_name = dependency.get('Name', '')\r\n                            if 'AudioBase' in dep_name:\r\n                                audio_plugins_string.append(dep_name)\r\n\r\n                except Exception as e:\r\n                    audio_plugins_string = [f\"Error: {str(e)}\"]"
    last_known_path:String=Get LastKnownPath() -> (LastKnownPath:String)
   ) -> (then, ReturnValue, has_changed:bool=~, audio_plugins_string:Array of Strings, new_last_path:String, content_browser_location:String) {inputs [last_known_path], outputs [has_changed, audio_plugins_string, new_last_path, content_browser_location]}
   > Branch(Condition=@n1.has_changed)
   > Set LastKnownPath(LastKnownPath:String=@n1.new_last_path) -> (Output_Get:String)
   > ClearListItems(self=Get CommonListView_121!(self:EUW WH Audio Base List Object Reference=Get EUW_WH_AudioBaseList() -> (EUW_WH_AudioBaseList:EUW WH Audio Base List Object Reference)) -> (CommonListView_121:Common List View Object Reference))
   > SetPath!(execute, self:EUW WH Audio Base List Object Reference=Get EUW_WH_AudioBaseList(), NewPath:String=@n1.content_browser_location) -> (then)
   > n2 = Branch(Condition=Is Empty!(TargetArray=@n1.audio_plugins_string) -> (ReturnValue:bool))
   > n3 = Construct Object from Class!(
    execute
    Class=/FortTechAudioTools/Widgets/Workflow/SubWidgets/AudioBaseTracker/WBP_WH_AudioBasePlugin_Entry.WBP_WH_AudioBasePlugin_Entry_C
    self
    NewName:Text="LOCTEXT(\"No AudioBase plugin found\")"
   ) -> (then, ReturnValue:WBP WH Audio Base Plugin Entry Object Reference)
   > AddItem!(execute, self:List View Object Reference=Get CommonListView_121(self=Get EUW_WH_AudioBaseList()), Item:Object Reference=@n3) -> (then)
   n4 = For Each Loop(Exec=@n2.else, Array=@n1.audio_plugins_string)
   > n5 = Construct Object from Class(
    Class=&WBP_WH_AudioBasePlugin_Entry_C
    NewName="To Text (String)"(InString=@n4)
   )
   > AddItem(self=Get CommonListView_121(self=Get EUW_WH_AudioBaseList()), Item=@n5)
   Event Construct()
   > Set Timer by Event(Delegate=Create Event(), Time=2, bLooping=true, bMaxOncePerFrame=true)
   > LoadSavedLayout()
   Event PreConstruct()
   Event Tick()
  ]}
 Handle_TabClicked {nodes [
   n1 = Handle_TabClicked!() -> (then, TabID:Guid Structure) {is=title K2Node_FunctionEntry}
   > ActivateTab(TabID=@n1)
  ]}
 Handle_TabClosed {nodes [
   n1 = Handle_TabClosed!() -> (then, TabID:Guid Structure) {is=title K2Node_FunctionEntry}
   > n2 = Keys!(execute, TargetMap:Map of Guid Structures to Struct WH Tab Info Structures=Get TabsByGuid()) -> (then, Keys:Array of Guid Structures)
   > Set ClosedTabIndex(ClosedTabIndex:int=Find Item!(TargetArray:Array of Guid Structures=@n2, ItemToFind:"Guid Structure (by ref)"=Get TabID()) -> (ReturnValue)) -> (Output_Get:int=0)
   n3 = Break Struct WH Tab Info(Struct_WH_TabInfo=Find(TargetMap=Get TabsByGuid(), Key=@n1))
   > Set IsSelectedTab(IsSelectedTab:bool=Get Selected!(self:WBP WH Tab Button Object Reference=@n3.TabButtonWidget_8_#3) -> (Selected:bool)) -> (Output_Get:bool)
   > RemoveChild(self=Get WidgetSwitcher_0(), Content=@n3.ContentWidget_12_#2)
   > RemoveEntry!(
    execute
    self:Dynamic Entry Box Object Reference=Get Tabs()
    EntryWidget:User Widget Object Reference=@n3.TabButtonWidget_8_#3
   ) -> (then)
   > Remove!(execute, TargetMap:Map of Guid Structures to Struct WH Tab Info Structures=Get TabsByGuid(), Key:"Guid Structure (by ref)"=@n3) -> (then, ReturnValue:bool)
   > SaveLayout()
   > Branch(Condition=Get IsSelectedTab() -> (IsSelectedTab:bool))
   > n4 = Keys(TargetMap=Get TabsByGuid())
   > Branch(Condition=Is Not Empty!(TargetArray:Array of Guid Structures=@n4) -> (ReturnValue:bool))
   n5 = Get ClosedTabIndex() -> (ClosedTabIndex:int=0)
   > n6 = Branch(Condition=Is Valid Index!(TargetArray:Array of Guid Structures=@n4, IndexToTest:int=@n5) -> (ReturnValue:bool))
   > ActivateTab(TabID="Get (a copy)"(Array:Array of Guid Structures=@n4, Dimension 1=@n5) -> (Output:Guid Structure))
   n7 = int - int(A=@n5)
   Branch(execute=@n6.else, Condition=Is Valid Index(TargetArray=@n4, IndexToTest=@n7))
   > ActivateTab(TabID="Get (a copy)"(Array=@n4, Dimension 1=@n7))
  ]}
 LoadSavedLayout {nodes [
   LoadSavedLayout() {is=title K2Node_FunctionEntry}
   > n1 = DoesSaveGameExist!(execute, SlotName:String=Get SaveGameSlotName() -> (SaveGameSlotName:String), UserIndex:int=0) -> (then, ReturnValue:bool)
   > n2 = Branch(Condition=@n1)
   > n3 = LoadGameFromSlot!(execute, SlotName:String=Get SaveGameSlotName(), UserIndex:int=0) -> (then, ReturnValue:Save Game Object Reference)
   > n4 = Cast To WH_SaveGameConfig!(execute, Object=@n3) -> (then, CastFailed, AsWH Save Game Config:WH Save Game Config Object Reference) {is=title K2Node_DynamicCast}
   > n5 = Set SaveObj(SaveObj:WH Save Game Config Object Reference=@n4) -> (Output_Get:WH Save Game Config Object Reference)
   > n6 = For Each Loop(Array:Array of Struct WH Tab Save Structures=Get Tabs!(self:WH Save Game Config Object Reference=@n5) -> (Tabs:Array of Struct WH Tab Save Structures)) -> (Array Element:Struct WH Tab Save Structure)
   n7 = Break Struct WH Tab Save!(Struct_WH_Tab_Save:"Struct WH Tab Save Structure (by ref)"=@n6) -> (
    TabID_2_#7:Guid Structure
    TabName_11_#8:Text
    DataAssetRef_9_#9:Soft Object Reference
   ) {is=title K2Node_BreakStruct}
   > AddTab(InGuid=@n7)
   > Branch(Condition=IsValidSoftObjectReference(SoftObjectReference=@n7.DataAssetRef_9_#9))
   > n8 = LoadAsset_Blocking(Asset=@n7.DataAssetRef_9_#9)
   n9 = Find(TargetMap=Get TabsByGuid(), Key=@n7)
   > Branch(Condition=@n9.ReturnValue)
   > n10 = Set members in Struct WH Tab Info!(
    execute:Exec
    StructRef:"Struct WH Tab Info Structure (by ref)"=@n9
    AssignedDataAsset_15_#4:Object Reference=@n8
   ) -> (then:Exec, StructOut:"Struct WH Tab Info Structure (by ref)") {is=title K2Node_SetFieldsInStruct}
   > Add(TargetMap=Get TabsByGuid(), Key=@n7, Value=@n10)
   > SetPropertyView(self=Break Struct WH Tab Info(Struct_WH_TabInfo=@n9).ContentWidget_12_#2, NewAsset=@n8)
   > n11 = Cast To BP_WorkflowHelperData!(execute, Object=@n8) -> (then, CastFailed, AsBP Workflow Helper Data:BP Workflow Helper Data Object Reference) {is=title K2Node_DynamicCast}
   > n12 = Branch(Condition=IsValid(Object=@n8))
   n13 = Break Struct WH Tab Info(Struct_WH_TabInfo=@n10)
   > SetTabName!(
    execute
    self:WBP WH Tab Button Object Reference=@n13.TabButtonWidget_8_#3
    NewName:Text=Get DisplayName!(self:BP Workflow Helper Data Object Reference=@n11) -> (DisplayName:Text)
   ) -> (then)
   ClearTabName(execute=@n12.else, self=@n13.TabButtonWidget_8_#3)
   n14 = CreateSaveGameObject!(execute=@n2.else, SaveGameClass:Save Game Class Reference=/FortTechAudioTools/Widgets/Workflow/Data/WH_SaveGameConfig.WH_SaveGameConfig_C) -> (then, ReturnValue:WH Save Game Config Object Reference)
   > Set SaveObj(SaveObj=@n14)
  ]}
 OnDataAssetChanged {nodes [
   n1 = OnDataAssetChanged!() -> (then, Data:Object Reference, TabID:Guid Structure) {is=title K2Node_FunctionEntry}
   n2 = Find(TargetMap=Get TabsByGuid(), Key=Get TabID())
   > Branch(Condition=@n2.ReturnValue)
   > n3 = Set members in Struct WH Tab Info(StructRef=@n2, AssignedDataAsset_15_#4=@n1)
   > Add(TargetMap=Get TabsByGuid(), Key=Get TabID(), Value=@n3)
   > n4 = Cast To BP_WorkflowHelperData(Object=@n1)
   n5 = Break Struct WH Tab Info(Struct_WH_TabInfo=@n3)
   > n6 = SetTabName(self=@n5.TabButtonWidget_8_#3, NewName=Get DisplayName(self=@n4))
   SaveLayout(execute=[@n7.then, @n6.then])
   n7 = ClearTabName(execute=@n4.CastFailed, self=@n5.TabButtonWidget_8_#3)
  ]}
 SaveLayout {nodes [
   SaveLayout() {is=title K2Node_FunctionEntry}
   > n1 = For Each Loop(Array:Array of User Widget Object References=GetAllEntries!(self:Dynamic Entry Box Base Object Reference=Get Tabs!() -> (Tabs:Dynamic Entry Box Object Reference)) -> (ReturnValue:Array of User Widget Object References)) -> (Array Element:User Widget Object Reference)
   n2 = Get SaveObj() -> (SaveObj:WH Save Game Config Object Reference)
   Set Tabs!(
    execute=@n1.Completed
    Tabs:Array of Struct WH Tab Save Structures=Get TabSaveArray() -> (TabSaveArray:Array of Struct WH Tab Save Structures)
    self:WH Save Game Config Object Reference=@n2
   ) -> (then, Output_Get:Array of Struct WH Tab Save Structures)
   > SaveGameToSlot!(execute, SaveGameObject:Save Game Object Reference=@n2, SlotName:String=Get SaveGameSlotName(), UserIndex:int=0) -> (then, ReturnValue:bool)
   n3 = Cast To WBP_WH_TabButton(execute=@n1.LoopBody, Object=@n1)
   n4 = Find(TargetMap=Get TabsByGuid(), Key=Get TabID!(self:WBP WH Tab Button Object Reference=@n3) -> (TabID))
   > Branch(Condition=@n4.ReturnValue)
   n5 = Break Struct WH Tab Info(Struct_WH_TabInfo=@n4)
   n6 = IsValid(Object=@n5.AssignedDataAsset_15_#4)
   n7 = Make Soft Reference!(Input:Object Reference=@n5.AssignedDataAsset_15_#4) -> (Output:Soft Object Reference)
   n8 = Get ButtonText!(self:WBP WH Tab Button Object Reference=@n5.TabButtonWidget_8_#3) -> (ButtonText:Common Text Object Reference)
   > Add(
    TargetArray:Array of Struct WH Tab Save Structures=Get TabSaveArray()
    NewItem:"Struct WH Tab Save Structure (by ref)"=Make Struct WH Tab Save!(
     TabID_2_#7:Guid Structure=@n5
     TabName_11_#8:Text=Get Text(self:Text Object Reference=@n8)
     DataAssetRef_9_#9:Soft Object Reference=Select(Option 0:Soft Object Reference, Option 1:Soft Object Reference=@n7, Index=@n6) -> (ReturnValue:Soft Object Reference)
    ) -> (Struct_WH_Tab_Save:Struct WH Tab Save Structure) {is=title K2Node_MakeStruct}
   ) {is=title K2Node_CallArrayFunction}
  ]}
]
variables [
 ActiveTabID {type=integer}
 LastKnownPath {type=string, default=""}
 SaveGameSlotName {type=string, default=EUW_Workflow_Helper_Layout}
 SaveObj {type=WH_SaveGameConfig_C, default=null}
 TabCounter {type=integer}
 TabsByGuid {default {}}
]
widgets {
 info {rootWidgetClass=&CanvasPanel, widgetCount=20}
 widgets [
  CanvasPanel_30 {widgetClassPath=&CanvasPanel, widget=CanvasPanel_30}
  VerticalBox_0 {
   widgetClassPath=&VerticalBox
   parent=CanvasPanel_30
   widget=VerticalBox_0
   slot {
    class=CanvasPanelSlot
    properties {bAutoSize=true, layoutData {alignment=<0 0>, anchors {maximum=<1 1>, minimum=<0 0>}, offsets=<ltrb 0 0 0 0>}}
   }
  }
  SizeBox_0 {
   widgetClassPath=&SizeBox
   parent=VerticalBox_0
   widget=SizeBox_0
   properties {bOverride_HeightOverride=true, heightOverride=100}
   slot {class=VerticalBoxSlot, properties {padding=<ltrb 10 10 10 10>}}
  }
  Border_1 {
   widgetClassPath=&Border
   parent=SizeBox_0
   widget=Border_1
   properties {
    background {
     bIsDynamicallyLoaded=false
     drawAs=Border
     imageType=NoImage
     mirroring=NoMirror
     resourceName=None
     resourceObject=null
     tiling=NoTile
     imageSize=<32 32>
     margin=<ltrb 1 1 1 1>
     outlineSettings {
      bUseBrushTransparency=false
      roundingType=HalfHeightRadius
      width=0
      color {colorUseRule=UseColor_Specified, specifiedColor=<rgba 0 0 0 0>}
      cornerRadii=<0 0 0 0>
     }
     tintColor {colorUseRule=UseColor_Specified, specifiedColor=<rgba 1 1 1 0>}
     uVRegion {bIsValid=false, max=<0 0>, min=<0 0>}
    }
    padding=<ltrb 0 0 0 0>
   }
   slot {class=SizeBoxSlot, properties {padding=<ltrb 0 0 0 -10>}}
  }
  EUW_WH_AudioBaseList {
   widgetClassPath=/FortTechAudioTools/Widgets/Workflow/SubWidgets/AudioBaseTracker/EUW_WH_AudioBaseList
   bIsVariable=true
   parent=Border_1
   widget=EUW_WH_AudioBaseList
   slot {class=BorderSlot, properties {padding=<ltrb 0 0 0 0>}}
  }
  Border_2 {
   widgetClassPath=&Border
   parent=VerticalBox_0
   widget=Border_2
   properties {
    background {
     bIsDynamicallyLoaded=false
     drawAs=Image
     imageType=NoImage
     mirroring=NoMirror
     resourceName=None
     resourceObject=null
     tiling=NoTile
     imageSize=<32 32>
     margin=<ltrb 0 0 0 0>
     outlineSettings {
      bUseBrushTransparency=false
      roundingType=HalfHeightRadius
      width=0
      color {colorUseRule=UseColor_Specified, specifiedColor=<rgba 0 0 0 0>}
      cornerRadii=<0 0 0 0>
     }
     tintColor {colorUseRule=UseColor_Specified, specifiedColor=<rgba 0.0103 0.0103 0.0103 1>}
     uVRegion {bIsValid=false, max=<0 0>, min=<0 0>}
    }
    padding=<ltrb 5 0 5 0>
   }
   slot {class=VerticalBoxSlot, properties {padding=<ltrb 10 10 10 0>, size {sizeRule=Fill, value=1}}}
  }
  VerticalBox_1 {widgetClassPath=&VerticalBox, parent=Border_2, widget=VerticalBox_1, slot {class=BorderSlot, properties {padding=<ltrb 5 0 5 0>}}}
  CanvasPanel_0 {
   widgetClassPath=&CanvasPanel
   parent=VerticalBox_1
   widget=CanvasPanel_0
   slot {class=VerticalBoxSlot, properties {padding=<ltrb -5 0 -5 0>}}
  }
  HorizontalBox_116 {
   widgetClassPath=&HorizontalBox
   parent=CanvasPanel_0
   widget=HorizontalBox_116
   slot {
    class=CanvasPanelSlot
    properties {bAutoSize=true, layoutData {alignment=<0 0>, anchors {maximum=<1 1>, minimum=<0 0>}, offsets=<ltrb 0 0 0 0>}}
   }
  }
  SizeBox_4 {
   widgetClassPath=&SizeBox
   parent=HorizontalBox_116
   widget=SizeBox_4
   properties {bOverride_HeightOverride=true, bOverride_WidthOverride=true, heightOverride=25, widthOverride=50}
   slot {class=HorizontalBoxSlot, properties {horizontalAlignment=HAlign_Left, verticalAlignment=VAlign_Top, padding=<ltrb 0 5 0 0>}}
  }
  Button_AddTab {
   widgetClassPath=&EditorUtilityButton
   parent=SizeBox_4
   widget=Button_AddTab
   properties {widgetStyle {
     clickedSlateSound {resourceObject=null}
     disabled {
      bIsDynamicallyLoaded=false
      drawAs=RoundedBox
      imageType=NoImage
      mirroring=NoMirror
      resourceName=None
      resourceObject=null
      tiling=NoTile
      imageSize=<0 0>
      margin=<ltrb 0 0 0 0>
      outlineSettings {
       bUseBrushTransparency=false
       roundingType=FixedRadius
       width=1
       color {colorUseRule=UseColor_ColorTable, specifiedColor=<rgba 0 0 0 0>}
       cornerRadii=<4 4 4 4>
      }
      tintColor {colorUseRule=UseColor_Specified, specifiedColor=<rgba 0.0395 0.0395 0.0395 1>}
      uVRegion {bIsValid=false, max=<0 0>, min=<0 0>}
     }
     disabledForeground {colorUseRule=UseColor_ColorTable, specifiedColor=<rgba 0 0 0 0>}
     hovered {
      bIsDynamicallyLoaded=false
      drawAs=Box
      imageType=NoImage
      mirroring=NoMirror
      resourceName=None
      resourceObject=null
      tiling=NoTile
      imageSize=<32 32>
      margin=<ltrb 0 0 0 0>
      outlineSettings {
       bUseBrushTransparency=false
       roundingType=FixedRadius
       width=1
       color {colorUseRule=UseColor_ColorTable, specifiedColor=<rgba 0 0 0 0>}
       cornerRadii=<4 4 4 4>
      }
      tintColor {colorUseRule=UseColor_Specified, specifiedColor=<rgba 0.1073 0.1073 0.1073 1>}
      uVRegion {bIsValid=false, max=<0 0>, min=<0 0>}
     }
     hoveredForeground {colorUseRule=UseColor_ColorTable, specifiedColor=<rgba 0 0 0 0>}
     hoveredSlateSound {resourceObject=null}
     normal {
      bIsDynamicallyLoaded=false
      drawAs=Box
      imageType=NoImage
      mirroring=NoMirror
      resourceName=None
      resourceObject=null
      tiling=NoTile
      imageSize=<32 32>
      margin=<ltrb 0 0 0 0>
      outlineSettings {
       bUseBrushTransparency=false
       roundingType=FixedRadius
       width=1
       color {colorUseRule=UseColor_ColorTable, specifiedColor=<rgba 0 0 0 0>}
       cornerRadii=<4 4 4 4>
      }
      tintColor {colorUseRule=UseColor_Specified, specifiedColor=<rgba 0.0212 0.0212 0.0212 1>}
      uVRegion {bIsValid=false, max=<0 0>, min=<0 0>}
     }
     normalForeground {colorUseRule=UseColor_ColorTable, specifiedColor=<rgba 0 0 0 0>}
     normalPadding=<ltrb 12 1.5 12 1.5>
     pressed {
      bIsDynamicallyLoaded=false
      drawAs=RoundedBox
      imageType=NoImage
      mirroring=NoMirror
      resourceName=None
      resourceObject=null
      tiling=NoTile
      imageSize=<32 32>
      margin=<ltrb 0 0 0 0>
      outlineSettings {
       bUseBrushTransparency=false
       roundingType=FixedRadius
       width=1
       color {colorUseRule=UseColor_ColorTable, specifiedColor=<rgba 0 0 0 0>}
       cornerRadii=<4 4 4 4>
      }
      tintColor {colorUseRule=UseColor_Specified, specifiedColor=<rgba 0.0284 0.0284 0.0284 1>}
      uVRegion {bIsValid=false, max=<0 0>, min=<0 0>}
     }
     pressedForeground {colorUseRule=UseColor_ColorTable, specifiedColor=<rgba 0 0 0 0>}
     pressedPadding=<ltrb 12 2.5 12 0.5>
     pressedSlateSound {resourceObject=null}
    }}
   slot {class=SizeBoxSlot, properties {padding=<ltrb 5 0 0 0>}}
  }
  TextBlock_77 {
   widgetClassPath=&TextBlock
   parent=Button_AddTab
   widget=TextBlock_77
   properties {
    text=+
    font {
     bForceMonospaced=false
     bMaterialIsStencil=false
     fontMaterial=null
     fontObject=&Roboto
     letterSpacing=0
     monospacedWidth=1
     size=18
     skewAmount=0
     typefaceFontName=Bold
     outlineSettings {
      bApplyOutlineToDropShadows=false
      bMiteredCorners=false
      bSeparateFillAlpha=false
      outlineMaterial=null
      outlineSize=0
      outlineColor=<rgba 0 0 0 1>
     }
    }
    renderTransform {angle=0, scale=<1 1>, shear=<0 0>, translation=<4 -2>}
   }
   slot {class=ButtonSlot, properties {horizontalAlignment=HAlign_Fill, verticalAlignment=VAlign_Fill, padding=<ltrb 0 0 0 0>}}
  }
  EditorUtilityScrollBox_44 {
   widgetClassPath=&EditorUtilityScrollBox
   parent=HorizontalBox_116
   widget=EditorUtilityScrollBox_44
   properties {orientation=Orient_Horizontal}
   slot {class=HorizontalBoxSlot, properties {padding=<ltrb 0 5 0 0>}}
  }
  Tabs {
   widgetClassPath=&DynamicEntryBox
   bIsVariable=true
   parent=EditorUtilityScrollBox_44
   widget=Tabs
   properties {entryWidgetClass=/FortTechAudioTools/Widgets/Workflow/SubWidgets/EntryWidgets/WBP_WH_TabButton.WBP_WH_TabButton_C, entrySpacing=<1 0>}
   slot {class=ScrollBoxSlot, properties {padding=<ltrb 1 0 0 0>}}
  }
  Border_193 {
   widgetClassPath=&Border
   parent=VerticalBox_1
   widget=Border_193
   properties {
    verticalAlignment=VAlign_Top
    background {
     bIsDynamicallyLoaded=false
     drawAs=Box
     imageType=NoImage
     mirroring=NoMirror
     resourceName=None
     resourceObject=null
     tiling=NoTile
     imageSize=<32 32>
     margin=<ltrb 0 0 0 0>
     outlineSettings {
      bUseBrushTransparency=false
      roundingType=HalfHeightRadius
      width=0
      color {colorUseRule=UseColor_Specified, specifiedColor=<rgba 0 0 0 0>}
      cornerRadii=<0 0 0 0>
     }
     tintColor {colorUseRule=UseColor_Specified, specifiedColor=<rgba 0.0212 0.0212 0.0212 1>}
     uVRegion {bIsValid=false, max=<0 0>, min=<0 0>}
    }
    padding=<ltrb 0 0 0 0>
   }
   slot {class=VerticalBoxSlot, properties {size {sizeRule=Fill, value=1}}}
  }
  WidgetSwitcher_0 {
   widgetClassPath=&WidgetSwitcher
   bIsVariable=true
   parent=Border_193
   widget=WidgetSwitcher_0
   slot {class=BorderSlot, properties {verticalAlignment=VAlign_Top, padding=<ltrb 0 0 0 0>}}
  }
  SizeBox_1 {
   widgetClassPath=&SizeBox
   parent=VerticalBox_0
   widget=SizeBox_1
   properties {bOverride_HeightOverride=true, heightOverride=30}
   slot {class=VerticalBoxSlot, properties {padding=<ltrb 10 10 10 10>}}
  }
  EditorUtilityButton_86 {
   widgetClassPath=&EditorUtilityButton
   parent=SizeBox_1
   widget=EditorUtilityButton_86
   properties {widgetStyle {
     clickedSlateSound {resourceObject=null}
     disabled {
      bIsDynamicallyLoaded=false
      drawAs=RoundedBox
      imageType=NoImage
      mirroring=NoMirror
      resourceName=None
      resourceObject=null
      tiling=NoTile
      imageSize=<0 0>
      margin=<ltrb 0 0 0 0>
      outlineSettings {
       bUseBrushTransparency=false
       roundingType=FixedRadius
       width=1
       color {colorUseRule=UseColor_ColorTable, specifiedColor=<rgba 0 0 0 0>}
       cornerRadii=<4 4 4 4>
      }
      tintColor {colorUseRule=UseColor_Specified, specifiedColor=<rgba 0.0395 0.0395 0.0395 1>}
      uVRegion {bIsValid=false, max=<0 0>, min=<0 0>}
     }
     disabledForeground {colorUseRule=UseColor_ColorTable, specifiedColor=<rgba 0 0 0 0>}
     hovered {
      bIsDynamicallyLoaded=false
      drawAs=RoundedBox
      imageType=NoImage
      mirroring=NoMirror
      resourceName=None
      resourceObject=null
      tiling=NoTile
      imageSize=<32 32>
      margin=<ltrb 0 0 0 0>
      outlineSettings {
       bUseBrushTransparency=false
       roundingType=FixedRadius
       width=1
       color {colorUseRule=UseColor_ColorTable, specifiedColor=<rgba 0 0 0 0>}
       cornerRadii=<4 4 4 4>
      }
      tintColor {colorUseRule=UseColor_Specified, specifiedColor=<rgba 0.0953 0.0953 0.0953 1>}
      uVRegion {bIsValid=false, max=<0 0>, min=<0 0>}
     }
     hoveredForeground {colorUseRule=UseColor_ColorTable, specifiedColor=<rgba 0 0 0 0>}
     hoveredSlateSound {resourceObject=null}
     normal {
      bIsDynamicallyLoaded=false
      drawAs=RoundedBox
      imageType=NoImage
      mirroring=NoMirror
      resourceName=None
      resourceObject=null
      tiling=NoTile
      imageSize=<32 32>
      margin=<ltrb 0 0 0 0>
      outlineSettings {
       bUseBrushTransparency=false
       roundingType=FixedRadius
       width=1
       color {colorUseRule=UseColor_ColorTable, specifiedColor=<rgba 0 0 0 0>}
       cornerRadii=<4 4 4 4>
      }
      tintColor {colorUseRule=UseColor_Specified, specifiedColor=<rgba 0.0284 0.0284 0.0284 1>}
      uVRegion {bIsValid=false, max=<0 0>, min=<0 0>}
     }
     normalForeground {colorUseRule=UseColor_ColorTable, specifiedColor=<rgba 0 0 0 0>}
     normalPadding=<ltrb 12 1.5 12 1.5>
     pressed {
      bIsDynamicallyLoaded=false
      drawAs=RoundedBox
      imageType=NoImage
      mirroring=NoMirror
      resourceName=None
      resourceObject=null
      tiling=NoTile
      imageSize=<32 32>
      margin=<ltrb 0 0 0 0>
      outlineSettings {
       bUseBrushTransparency=false
       roundingType=FixedRadius
       width=1
       color {colorUseRule=UseColor_ColorTable, specifiedColor=<rgba 0 0 0 0>}
       cornerRadii=<4 4 4 4>
      }
      tintColor {colorUseRule=UseColor_Specified, specifiedColor=<rgba 0.0284 0.0284 0.0284 1>}
      uVRegion {bIsValid=false, max=<0 0>, min=<0 0>}
     }
     pressedForeground {colorUseRule=UseColor_ColorTable, specifiedColor=<rgba 0 0 0 0>}
     pressedPadding=<ltrb 12 2.5 12 0.5>
     pressedSlateSound {resourceObject=null}
    }}
   slot {class=SizeBoxSlot, properties {horizontalAlignment=HAlign_Left}}
  }
  ScaleBox_0 {
   widgetClassPath=&ScaleBox
   parent=EditorUtilityButton_86
   widget=ScaleBox_0
   properties {stretchDirection=DownOnly}
   slot {class=ButtonSlot}
  }
  CommonTextBlock_46 {
   widgetClassPath=&CommonTextBlock
   parent=ScaleBox_0
   widget=CommonTextBlock_46
   properties {
    style=/FortUILibrary/Styles/Typography/Foundation/TextStyle-Base-S.TextStyle-Base-S_C
    text=Workflow Helper Documentation
    font {
     bForceMonospaced=false
     bMaterialIsStencil=false
     fontMaterial=null
     fontObject=&NotoSans
     letterSpacing=0
     monospacedWidth=1
     size=18
     skewAmount=0
     typefaceFontName=NotoSans-Regular
     outlineSettings {
      bApplyOutlineToDropShadows=false
      bMiteredCorners=false
      bSeparateFillAlpha=false
      outlineMaterial=null
      outlineSize=0
      outlineColor=<rgba 0 0 0 1>
     }
    }
    shadowOffset=<0 0>
   }
   slot {class=ScaleBoxSlot}
  }
 ]
}
